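; Flower box group building prop that adds a second, offset copy of its static mesh in the construction script; event graph logic is empty.
; If building texture data exists, the first entry's tint colors both the stand and the copy.
; The same loop adds instances to the instanced mesh component, based on instance 0, shifted along X by the secondary offset with seeded random yaw.
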(event EventBeginPlay)

(event Collision|EventActorBeginOverlap
  (params (OtherActor :type Actor%20Object%20Reference)))

(event EventTick
  (params (DeltaSeconds :type Float%20%28single-precision%29)))

(fn ConstructionScript
  (|CallParentFunction)
  (bind _returnvalue (AddComponent|NODE_AddStaticMeshComponent-1 :RelativeTransform_Location (Math|Vector|MakeVector (Variables|Default|GetSecondaryOffset)) :RelativeTransform_Rotation "0, 0, 0" :RelativeTransform_Scale "1.000000,1.000000,1.000000"))
  (bind _output_get (Variables|Default|SetExtraMesh _returnvalue))
  (Components|StaticMesh|SetStaticMesh _output_get (Variables|/Script/Engine.StaticMeshComponent|GetStaticMesh (Variables|Debug|GetStaticMeshComponent)))
  (bind _outtexturedataarray (Building|GetBuildingTextureData))
  (if (> (Utilities|Array|Length _outtexturedataarray) 0)
    (bind _output_get_2 (Variables|Default|SetTint%20Color_0 (Variables|/Script/FortniteGame.BuildingTextureData|GetTintColor (Utilities|Array|Get(acopy) _outtexturedataarray))))
    ;[1 1 1] can't mismatch ISM  and SMs
    (for _array_element (Utilities|Array|MakeArray (Variables|Default|GetSM_Outdoor_FlowerBoxStand_A) (Variables|Default|GetExtraMesh))
      (Default|SetColoronSecondaryMesh (Math|Conversions|ToLinearColor(Color) _output_get_2) _array_element)
      (Components|InstancedStaticMesh|AddInstance (Actor|GetComponentbyClass "/Script/Engine.InstancedStaticMeshComponent") (Math|Transform|MakeTransform (Math|Vector|MakeVector (+ (.x (.location (Components|InstancedStaticMesh|GetInstanceTransform (Actor|GetComponentbyClass "/Script/Engine.InstancedStaticMeshComponent") 0 false))) (Variables|Default|GetSecondaryOffset)) (.y (.location (Components|InstancedStaticMesh|GetInstanceTransform (Actor|GetComponentbyClass "/Script/Engine.InstancedStaticMeshComponent") 0 false))) (.z (.location (Components|InstancedStaticMesh|GetInstanceTransform (Actor|GetComponentbyClass "/Script/Engine.InstancedStaticMeshComponent") 0 false)))) (Math|Rotator|MakeRotator 0.0 0.0 (Math|Random|RandomFloatinRangefromStream (Variables|Default|GetCurrentRandomInstanceStream) 0.0 9e+01)) (.scale (Components|InstancedStaticMesh|GetInstanceTransform (Actor|GetComponentbyClass "/Script/Engine.InstancedStaticMeshComponent") 0 false))) false))))
